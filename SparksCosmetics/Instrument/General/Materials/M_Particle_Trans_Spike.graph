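# Translucent lit material for Niagara sprite spike particles in Sparks instrument cosmetics, shaped by a cartoon impact texture.
# Particle color is multiplied by randomly offset circle noise, a stepped outline darkening from the impact G channel, and a repeating vertical falloff.
# Time-of-day tint and dynamic emissive intensity scale that color before an emissive adjustment function.
# Opacity subtracts dissolve-shifted noise from the impact R channel, then multiplies it for harder edges.
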Material /SparksCosmetics/Instrument/General/Materials/M_Particle_Trans_Spike
graph {
 properties {bAutomaticallySetUsageInEditor=false, bUsedWithNiagaraSprites=true, blendMode=BLEND_Translucent, usedShadingModels=Default Lit}
 nodes [
  n1 = TextureSample(UVs=Add(A=TextureCoordinate() {uTiling=0.75, vTiling=0.75}, B=ParticleRandom())) {texture=/Game/Effects/Fort_Effects/Textures/Tile/T_CircleNoise_01_M}
  n2 = LinearInterpolate(A=ScalarParameter() {defaultValue=0.1, parameterName=Minimum Texture Value}, Alpha=@n1.R)
  n3 = TextureSample() {texture=/Game/Effects/Fort_Effects/Textures/Impacts/T_FX_CartoonImpacts_02_M}
  n4 = DynamicParameter() {defaultValue=<rgba 0 0.688 0 0>, paramNames [Outline Intensity, Outline Size, Emissive Intensity, Dissolve]}
  n5 = Step(Y=@n3.G, X=OneMinus(None=@n4.Outline Size))
  n6 = OneMinus(None=Multiply(A=OneMinus(None=@n5), B=@n4.Outline Intensity))
  n7 = TextureCoordinate()
  n8 = Frac(None=Multiply(A=ComponentMask(None=@n7) {g=true}) {constB=2})
  n9 = Multiply(A=Multiply(A=ParticleColor().RGB, B=Multiply(A=@n2, B=@n6)), B=Power(Base=OneMinus(None=@n8)))
  n10 = MaterialFunctionCall(
   Color_Morning=VectorParameter() {group=TOD, parameterName=TOD Tint Morning, sortPriority=1, defaultValue=<rgba 1 0.5276 0.3138 1>}.RGB
   Color_Day=VectorParameter() {group=TOD, parameterName=TOD Tint Day, sortPriority=2, defaultValue=<rgba 1 1 1 1>}.RGB
   Color_Evening=VectorParameter() {group=TOD, parameterName=TOD Tint Evening, sortPriority=3, defaultValue=<rgba 1 0.5276 0.3138 1>}.RGB
   Color_Night=VectorParameter() {group=TOD, parameterName=TOD Tint Night, sortPriority=4, defaultValue=<rgba 0.651 0.7228 1 1>}.RGB
  ) {materialFunction=&Color_Time_of_Day}
  n11 = DynamicParameter() {defaultValue=<rgba 0 0.688 0 0>, paramNames [Outline Intensity, Outline Size, Emissive Intensity, Dissolve]}
  MaterialFunctionCall(Emissive=Multiply(A=Multiply(A=@n9, B=@n10.Result), B=@n11.Emissive Intensity)) {materialFunction=&MF_FX_EmissiveAdjustments}
  n12 = DynamicParameter() {defaultValue=<rgba 0 0.688 0 0>, paramNames [Outline Intensity, Outline Size, Emissive Intensity, Dissolve]}
  n13 = Saturate(None=Add(A=@n1.R, B=LinearInterpolate(Alpha=@n12.Dissolve) {constA=-1}))
  Reroute(None=Multiply(A=Subtract(A=@n3.R, B=@n13), B=ScalarParameter() {defaultValue=5, parameterName=Opacity Multiplier}))
 ]
}
